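; Music encounter Blueprint for the MaskMaker boss that drives combat music phases and audio parameters.
; On combat entry it sets music attenuation from the encounter volume radius and a fixed falloff.
; Below half health, health changes move any nonzero phase to the final phase and feed the Boss.Hp sweetener parameter.
; Named events select the stunned phase, or on recovery the base or final phase by health.
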(event EventEnterCombatState
  (|CallParentFunction)
  (bind _combat_music_audio_component (Variables|MusicEncounter|Boss|Combat|GetCombat%20Music%20Audio%20Component)
    (:then
      (Audio|Parameter|SetFloatParameter _combat_music_audio_component "MusicEncounter.Attenuation.Volume.InnerRadius" (Variables|MusicEncounter|Volume|GetRadius) :!class /Script/AudioExtensions.AudioParameterControllerInterface)
      (Audio|Parameter|SetFloatParameter _combat_music_audio_component "MusicEncounter.Attenuation.Volume.FalloffDistance" 2e+03 :!class /Script/AudioExtensions.AudioParameterControllerInterface))
    (:else)))

(event MusicEncounter|Boss|EventBossHealthChanged
  (params (Boss%20Component%20Interface :type Music%20Encounter%20Component%20Boss%20Interface%20Interface) (Health%20Info :type Struct%20NPC%20Health%20Info%20Structure))
  (|CallParentFunction Boss%20Component%20Interface Health%20Info)
  (bind _output_get (Variables|Default|SetHealth%20Percent (Utilities|Struct|BreakStructNPCHealthInfo Health%20Info)))
  (Math|Float|CompareFloat _output_get 0.5
    (:<
      ;[0.6 0.4 0.8] Enter final Phase at 50% hp
      (Utilities|FlowControl|Sequence
        (:then_0
          (bind _combat_phase (MusicEncounterDashboard|Boss|GetBossEncounterCombatPhase))
          (if (!= 0 _combat_phase)
            (CallFunction|SetBossEncounterCombatPhase 2)))
        (:then_1
          ;[0.6 0.4 0.8] Update Boss sweetner layer when below 50% hp
          (bind _combat_music_audio_component (Variables|MusicEncounter|Boss|Combat|GetCombat%20Music%20Audio%20Component)
            (:then
              (Audio|Parameter|SetFloatParameter _combat_music_audio_component "Boss.Hp" (Variables|Default|GetHealth%20Percent) :!class /Script/AudioExtensions.AudioParameterControllerInterface))
            (:else)))))))

(event MusicEncounter|EventExecuteNamedEvent
  (params (Music%20Encounter%20Component :type Music%20Encounter%20Component%20Base%20Object%20Reference) (Event%20Name :type Name :default "None"))
  (switch name Event%20Name
    (:Stunned
      (CallFunction|SetBossEncounterCombatPhase 1))
    (:Not%20Stunned
      (CallFunction|SetBossEncounterCombatPhase (Math|Integer|SelectInt 2 0 (< (Variables|Default|GetHealth%20Percent) 0.5))))))
