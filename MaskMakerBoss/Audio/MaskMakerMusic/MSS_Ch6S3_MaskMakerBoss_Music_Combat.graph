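# Adaptive stereo combat music for the Chapter 6 Season 3 Mask Maker boss, driven by boss phase and tempo.
# Phase changes fire clock-quantized stingers, then delayed loop starts alternate between A/B decks by phase parity with envelope crossfades.
# Each deck mixes two distance-curved layers, one also modulated by a weapon-fire envelope follower.
# A near-death sweetener rises as boss HP drops, and pausing fades out the mix and finishes the sound.
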MetaSoundSource /MaskMakerBoss/Audio/MaskMakerMusic/MSS_Ch6S3_MaskMakerBoss_Music_Combat
inputs [
 AlertLevel {type=Float}
 Boss.Hp {type=Float}
 Gain.A {type=Float}
 Gain.B {type=Float}
 Gain.Final {type=Float}
 Gain.Stingers {type=Float}
 Gain.Sweetner {type=Float}
 MusicEncounter.Attenuation.Distance {type=Float}
 MusicEncounter.Attenuation.Volume.FalloffDistance {type=Float}
 MusicEncounter.Attenuation.Volume.InnerRadius {type=Float}
 MusicEncounter.bIsPaused {type=Bool}
 MusicEncounter.Boss.Phase {type=Int32}
 MusicEncounter.PauseFadeOutTime {type=Time}
 Tempo {type=Float}
 Time.ElapsedTime {type=Float}
 Time.GameTimeSeconds {type=Float}
 Time.PlayTime {type=Float}
 Time.TimeSinceLastPlayed {type=Float}
 UE.Source.OnPlay {type=Trigger}
]
interfaces [Fort.Time:1.0, UE.OutputFormat.Stereo:1.0, UE.Source.OneShot:1.0, UE.Source:1.1]
outputs [
 "UE.OutputFormat.Stereo.Audio:0" {type=Audio}
 "UE.OutputFormat.Stereo.Audio:1" {type=Audio}
 UE.Source.OneShot.OnFinished {type=Trigger}
]
pages [Default {
  input_defaults {
   AlertLevel=0
   Boss.Hp=0.112
   Gain.A=1
   Gain.B=1
   Gain.Final=1
   Gain.Stingers=1
   Gain.Sweetner=1
   MusicEncounter.Attenuation.Distance=0
   MusicEncounter.Attenuation.Volume.FalloffDistance=2000
   MusicEncounter.Attenuation.Volume.InnerRadius=6000
   MusicEncounter.Boss.Phase=0
   MusicEncounter.PauseFadeOutTime=5
   MusicEncounter.bIsPaused=null
   Tempo=120
   Time.ElapsedTime=0
   Time.GameTimeSeconds=0
   Time.PlayTime=0
   Time.TimeSinceLastPlayed=500
   UE.Source.OnPlay=null
  }
  nodes [
   n1 = HarmonixNodes.Metronome(Tempo=^Tempo, Transport=HarmonixNodes.TriggerToTransport(Play=^UE.Source.OnPlay).Transport)
   n2 = InitVariable.Trigger()
   TriggerOnValueChange_2 = UE.Trigger On Value Change:Int32(Value=^MusicEncounter.Boss.Phase)
   TriggerAnyTriggerAny2_0 = "TriggerAny.Trigger Any (2)"(In 1=^UE.Source.OnPlay, In 0=@TriggerOnValueChange_2.Trigger)
   TriggerOnce_0 = UE.Trigger Once(
    Trigger In=HarmonixNodes.MidiClockSubdivisionTrigger(MIDI Clock=@n1.MIDI Clock) {also {defaults {Base Grid Size=6}}}.Trigger Out
    Reset=VariableAccessor.Trigger(Variable=VariableMutator.Trigger(Variable=@n2.Variable, Value=@TriggerAnyTriggerAny2_0.Out).Variable).Value
   ) {also {defaults {Start Closed="<MetaSoundBuilderResult.SUCCEEDED: 0>"}}}
   n3 = VariableDeferredAccessor.Trigger(Variable=InitVariable.Trigger().Variable)
   TriggerOnce_3 = UE.Trigger Once(Trigger In=@TriggerOnce_0.Trigger Out, Reset=@n3.Value)
   n4 = VariableAccessor.WaveAsset:Array(Variable=InitVariable.WaveAsset:Array().Variable)
   n5 = VariableAccessor.WaveAsset:Array(Variable=@n4.Variable)
   n6 = UE.Subtract:Int32(PrimaryOperand=Array.Num:WaveAsset:Array(Array=@n5.Value).Num) {also {defaults {AdditionalOperands=1}}}
   n7 = VariableAccessor.Int32(Variable=VariableMutator.Int32(Variable=InitVariable.Int32().Variable, Value=Value.Value:Int32(Set=^UE.Source.OnPlay, Target Value=@n6.Out).Output Value).Variable)
   n8 = VariableAccessor.Int32(Variable=@n7.Variable)
   n9 = VariableAccessor.Int32(Variable=@n8.Variable)
   n10 = VariableAccessor.Int32(Variable=@n9.Variable)
   n11 = TriggerCompare.Clamp:Int32(A=^MusicEncounter.Boss.Phase, Compare=@TriggerOnce_3.Trigger Out, B=@n10.Value)
   TriggerRouteTriggerRouteFloat2_1 = "TriggerRoute.Trigger Route (Float, 2):Float"(Set 1=@n11.False, Set 0=@n11.True) {also {defaults {Value 0=1, Value 1=8}}}
   TriggerDelay_0 = UE.Trigger Delay(
    Delay Time=UE.BPMToSeconds(BPM=^Tempo, Beat Multiplier=@TriggerRouteTriggerRouteFloat2_1.Value).Seconds
    In=@TriggerRouteTriggerRouteFloat2_1.On Set
   )
   n12 = VariableAccessor.Trigger(Variable=VariableMutator.Trigger(Value=@TriggerDelay_0.Out, Variable=@n3.Variable).Variable)
   n13 = TriggerCompare.Clamp:Int32(
    A=^MusicEncounter.Boss.Phase
    Compare=TriggerCompare.Clamp:Int32(A=UE.Modulo:Int32(PrimaryOperand=^MusicEncounter.Boss.Phase) {also {defaults {AdditionalOperands=2}}}.Out, Compare=@n12.Value).True
    B=@n8.Value
   ) {also {defaults {Type=1}}}
   n14 = VariableDeferredAccessor.Trigger(Variable=InitVariable.Trigger().Variable)
   n15 = VariableAccessor.MIDIClock(Variable=VariableMutator.MIDIClock(Value=@n1.MIDI Clock, Variable=InitVariable.MIDIClock().Variable).Variable)
   n16 = VariableAccessor.MIDIClock(Variable=@n15.Variable)
   n17 = VariableAccessor.MIDIClock(Variable=@n16.Variable)
   n18 = HarmonixNodes.TransportWavePlayerController(Transport=HarmonixNodes.TriggerToTransport(Play=@n13.True, Stop=@n14.Value).Transport, MIDI Clock=@n17.Value) {also {version [0,1]}}
   n19 = VariableAccessor.Trigger(Variable=VariableMutator.Trigger(Value=@n18.Play, Variable=InitVariable.Trigger().Variable).Variable)
   n20 = VariableAccessor.Trigger(Variable=@n19.Variable)
   n21 = VariableAccessor.Trigger(Variable=@n20.Variable)
   n22 = VariableAccessor.Trigger(Variable=@n21.Variable)
   n23 = VariableAccessor.Trigger(Variable=VariableMutator.Trigger(Variable=InitVariable.Trigger().Variable, Value=@n13.False).Variable)
   n24 = VariableAccessor.Trigger(Variable=@n23.Variable)
   n25 = TriggerCompare.Clamp:Int32(
    A=^MusicEncounter.Boss.Phase
    Compare=TriggerCompare.Clamp:Int32(
     A=UE.Modulo:Int32(PrimaryOperand=^MusicEncounter.Boss.Phase) {also {defaults {AdditionalOperands=2}}}.Out
     Compare=VariableAccessor.Trigger(Variable=@n12.Variable).Value
    ) {also {defaults {B=0, Type=1}}}.True
    B=@n9.Value
   ) {also {defaults {Type=1}}}
   n26 = VariableAccessor.Trigger(Variable=VariableMutator.Trigger(Variable=InitVariable.Trigger().Variable, Value=@n25.False).Variable)
   n27 = VariableAccessor.Trigger(Variable=@n26.Variable)
   TriggerAnyTriggerAny3_0 = "TriggerAny.Trigger Any (3)"(In 0=@n22.Value, In 2=VariableAccessor.Trigger(Variable=@n24.Variable).Value, In 1=VariableAccessor.Trigger(Variable=@n27.Variable).Value)
   n28 = VariableDeferredAccessor.Trigger(Variable=InitVariable.Trigger().Variable)
   n29 = VariableAccessor.MIDIClock(Variable=@n17.Variable)
   n30 = HarmonixNodes.TransportWavePlayerController(Transport=HarmonixNodes.TriggerToTransport(Play=@n25.True, Stop=@n28.Value).Transport, MIDI Clock=@n29.Value) {also {version [0,1]}}
   n31 = VariableAccessor.Trigger(Variable=VariableMutator.Trigger(Value=@n30.Play, Variable=InitVariable.Trigger().Variable).Variable)
   n32 = VariableAccessor.Trigger(Variable=@n31.Variable)
   n33 = VariableAccessor.Trigger(Variable=@n32.Variable)
   n34 = VariableAccessor.WaveAsset:Array(Variable=InitVariable.WaveAsset:Array().Variable)
   n35 = VariableAccessor.WaveAsset:Array(Variable=@n34.Variable)
   WavePlayer_1 = UE.Wave Player:Stereo(
    Wave Asset=Array.Get:WaveAsset:Array(Index=^MusicEncounter.Boss.Phase, Trigger=@n33.Value, Array=@n35.Value).Element
    Start Time=@n30.Start Time
    Stop=@n30.Stop
    Play=@n31.Value
   ) {also {defaults {Loop="<MetaSoundBuilderResult.SUCCEEDED: 0>"}, version [1,1]}}
   n36 = VariableAccessor.Trigger(Variable=VariableMutator.Trigger(Variable=InitVariable.Trigger().Variable, Value=@WavePlayer_1.On Play).Variable)
   ADSREnvelopeADSREnvelope_2 = ADSR Envelope.ADSR Envelope:Float(
    Release Time=UE.BPMToSeconds(BPM=^Tempo) {also {defaults {Beat Multiplier=2}}}.Seconds
    Trigger Release=@TriggerAnyTriggerAny3_0.Out
    Trigger Attack=VariableAccessor.Trigger(Variable=@n36.Variable).Value
   ) {also {defaults {Sustain Level=1}}}
   DecibelstoLinearGain_2 = UE.Decibels to Linear Gain:Float(Decibels=MapRange.MapRange:Float(In=Modulation.GetModulatorValue() {also {defaults {Modulator=/Game/Sounds/Modulation/EnvelopeFollowers/FortWeapon/CB_WeaponFireEnvFollower}}}.Out) {also {defaults {Out Range A=-3, Out Range B=0}}}.Out Value)
   n37 = UE.InterpTo:Audio(Target=^MusicEncounter.Attenuation.Distance, Interp Time=UE.BPMToSeconds(BPM=^Tempo).Seconds)
   n38 = MapRange.MapRange:Float(In Range B=^MusicEncounter.Attenuation.Volume.InnerRadius, In=@n37.Value) {also {defaults {In Range A=500, Out Range A=1, Out Range B=0}}}
   n39 = VariableAccessor.Float(Variable=VariableMutator.Float(
     Variable=InitVariable.Float().Variable
     Value=UE.WaveTableBankEvaluate(Input=@n38.Out Value) {also {defaults {Index=1, WaveTableBank=&/WTB_MaskMaker_AttenuationCurve}}}.Output
    ).Variable)
   n40 = MapRange.MapRange:Float(
    In=@n37.Value
    In Range A=Min.Min:Float(B=UE.Subtract:Float(AdditionalOperands=^MusicEncounter.Attenuation.Volume.FalloffDistance, PrimaryOperand=^MusicEncounter.Attenuation.Volume.InnerRadius).Out) {also {defaults {A=4000}}}.Value
    In Range B=UE.Reroute(Value=^MusicEncounter.Attenuation.Volume.InnerRadius).Value
   ) {also {defaults {Out Range A=1, Out Range B=0}}}
   n41 = VariableAccessor.Float(Variable=VariableMutator.Float(
     Variable=InitVariable.Float().Variable
     Value=UE.WaveTableBankEvaluate(Input=@n40.Out Value) {also {defaults {WaveTableBank=&WTB_MaskMaker_AttenuationCurve}}}.Output
    ).Variable)
   n42 = VariableAccessor.Float(Variable=@n41.Variable)
   n43 = VariableAccessor.Trigger(Variable=@n33.Variable)
   n44 = VariableAccessor.WaveAsset:Array(Variable=InitVariable.WaveAsset:Array().Variable)
   n45 = VariableAccessor.WaveAsset:Array(Variable=@n44.Variable)
   WavePlayer_2 = UE.Wave Player:Stereo(
    Wave Asset=Array.Get:WaveAsset:Array(Index=^MusicEncounter.Boss.Phase, Trigger=@n43.Value, Array=@n45.Value).Element
    Start Time=@n30.Start Time
    Stop=@n30.Stop
    Play=@n32.Value
   ) {also {defaults {Loop="<MetaSoundBuilderResult.SUCCEEDED: 0>"}, version [1,1]}}
   AudioMixerAudioMixerStereo2_0 = "AudioMixer.Audio Mixer (Stereo, 2)"(
    Gain 1=UE.Multiply:Float(
     AdditionalOperands=UE.Multiply:Float(AdditionalOperands=@ADSREnvelopeADSREnvelope_2.Out Envelope, PrimaryOperand=@DecibelstoLinearGain_2.Linear Gain).Out
     PrimaryOperand=VariableAccessor.Float(Variable=@n39.Variable).Value
    ).Out
    Gain 0=UE.Multiply:Float(AdditionalOperands=@ADSREnvelopeADSREnvelope_2.Out Envelope, PrimaryOperand=@n42.Value).Out
    In 0 L=@WavePlayer_1.Out Left
    In 0 R=@WavePlayer_1.Out Right
    In 1 L=@WavePlayer_2.Out Left
    In 1 R=@WavePlayer_2.Out Right
   )
   TriggerAnyTriggerAny3_1 = "TriggerAny.Trigger Any (3)"(In 1=@n27.Value, In 2=@n24.Value, In 0=VariableAccessor.Trigger(Variable=@n43.Variable).Value)
   WavePlayer_3 = UE.Wave Player:Stereo(
    Wave Asset=Array.Get:WaveAsset:Array(Index=^MusicEncounter.Boss.Phase, Trigger=VariableAccessor.Trigger(Variable=@n22.Variable).Value, Array=@n34.Value).Element
    Start Time=@n18.Start Time
    Stop=@n18.Stop
    Play=@n19.Value
   ) {also {defaults {Loop="<MetaSoundBuilderResult.SUCCEEDED: 0>"}, version [1,1]}}
   n46 = VariableAccessor.Trigger(Variable=VariableMutator.Trigger(Variable=InitVariable.Trigger().Variable, Value=@WavePlayer_3.On Play).Variable)
   ADSREnvelopeADSREnvelope_1 = ADSR Envelope.ADSR Envelope:Float(
    Release Time=UE.BPMToSeconds(BPM=^Tempo) {also {defaults {Beat Multiplier=2}}}.Seconds
    Trigger Release=@TriggerAnyTriggerAny3_1.Out
    Trigger Attack=VariableAccessor.Trigger(Variable=@n46.Variable).Value
   ) {also {defaults {Sustain Level=1}}}
   DecibelstoLinearGain_1 = UE.Decibels to Linear Gain:Float(Decibels=MapRange.MapRange:Float(In=Modulation.GetModulatorValue() {also {defaults {Modulator=&CB_WeaponFireEnvFollower}}}.Out) {also {defaults {Out Range A=-3, Out Range B=0}}}.Out Value)
   WavePlayer_5 = UE.Wave Player:Stereo(
    Wave Asset=Array.Get:WaveAsset:Array(Index=^MusicEncounter.Boss.Phase, Trigger=@n21.Value, Array=@n44.Value).Element
    Start Time=@n18.Start Time
    Stop=@n18.Stop
    Play=@n20.Value
   ) {also {defaults {Loop="<MetaSoundBuilderResult.SUCCEEDED: 0>"}, version [1,1]}}
   AudioMixerAudioMixerStereo2_1 = "AudioMixer.Audio Mixer (Stereo, 2)"(
    Gain 0=UE.Multiply:Float(AdditionalOperands=@ADSREnvelopeADSREnvelope_1.Out Envelope, PrimaryOperand=@n41.Value).Out
    Gain 1=UE.Multiply:Float(
     AdditionalOperands=UE.Multiply:Float(AdditionalOperands=@ADSREnvelopeADSREnvelope_1.Out Envelope, PrimaryOperand=@DecibelstoLinearGain_1.Linear Gain).Out
     PrimaryOperand=@n39.Value
    ).Out
    In 0 L=@WavePlayer_3.Out Left
    In 0 R=@WavePlayer_3.Out Right
    In 1 L=@WavePlayer_5.Out Left
    In 1 R=@WavePlayer_5.Out Right
   )
   TriggerAnyTriggerAny2_2 = "TriggerAny.Trigger Any (2)"(In 0=@n46.Value, In 1=@n36.Value)
   n47 = TriggerCompare.Clamp:Int32(A=^MusicEncounter.Boss.Phase, Compare=@TriggerAnyTriggerAny2_2.Out, B=@n7.Value)
   n48 = InitVariable.Trigger()
   n49 = VariableDeferredAccessor.Trigger(Variable=InitVariable.Trigger().Variable)
   n50 = HarmonixNodes.TransportWavePlayerController(
    Transport=HarmonixNodes.TriggerToTransport(Stop=@n47.False, Play=@n47.True, Continue=VariableAccessor.Trigger(Variable=@n48.Variable).Value, Pause=@n49.Value).Transport
    MIDI Clock=VariableAccessor.MIDIClock(Variable=@n29.Variable).Value
   ) {also {version [0,1]}}
   WavePlayer_0 = UE.Wave Player:Stereo(Play=@n50.Play, Start Time=@n50.Start Time, Stop=@n50.Stop) {also {
     defaults {Loop="<MetaSoundBuilderResult.SUCCEEDED: 0>", Wave Asset=/MaskMakerBoss/Audio/MX_S36_Combat_MaskMaker_Phase_3_LP_NearDeath}
     version [1,1]
    }}
   n51 = &MSP_StereoGain(
    "Gain (Linear)"=UE.InterpTo:Audio(Target=MapRange.MapRange:Float(In=^Boss.Hp) {also {defaults {In Range A=0.2, In Range B=0.5, Out Range A=1, Out Range B=0}}}.Out Value).Value
    Left In=@WavePlayer_0.Out Left
    Right In=@WavePlayer_0.Out Right
   )
   TriggerDelay_1 = UE.Trigger Delay(Delay Time=UE.BPMToSeconds(BPM=^Tempo).Seconds, In=@TriggerOnce_0.Trigger Out)
   TriggerOnce_1 = UE.Trigger Once(Reset=@TriggerDelay_1.Out, Trigger In=@TriggerOnce_0.Trigger Out)
   n52 = VariableAccessor.Trigger(Variable=VariableMutator.Trigger(Variable=InitVariable.Trigger().Variable, Value=@TriggerOnce_1.Trigger Out).Variable)
   n53 = VariableDeferredAccessor.Trigger(Variable=InitVariable.Trigger().Variable)
   n54 = HarmonixNodes.TransportWavePlayerController(
    Transport=HarmonixNodes.TriggerToTransport(Play=VariableAccessor.Trigger(Variable=@n52.Variable).Value, Stop=@n53.Value).Transport
    MIDI Clock=@n15.Value
   ) {also {version [0,1]}}
   WavePlayer_4 = UE.Wave Player:Stereo(
    Wave Asset=Array.Get:WaveAsset:Array(Index=^MusicEncounter.Boss.Phase, Trigger=@n52.Value, Array=@n4.Value).Element
    Play=@n54.Play
    Start Time=@n54.Start Time
    Stop=@n54.Stop
   ) {also {version [1,1]}}
   AudioMixerAudioMixerStereo4_0 = "AudioMixer.Audio Mixer (Stereo, 4)"(
    Gain 0=^Gain.A
    Gain 1=^Gain.B
    Gain 2=^Gain.Stingers
    Gain 3=^Gain.Sweetner
    In 1 L=@AudioMixerAudioMixerStereo2_0.Out L
    In 1 R=@AudioMixerAudioMixerStereo2_0.Out R
    In 0 L=@AudioMixerAudioMixerStereo2_1.Out L
    In 0 R=@AudioMixerAudioMixerStereo2_1.Out R
    In 3 L=@n51.Left Out
    In 3 R=@n51.Right Out
    In 2 L=@WavePlayer_4.Out Left
    In 2 R=@WavePlayer_4.Out Right
   )
   n55 = InitVariable.Trigger()
   TriggerOnValueChange_0 = UE.Trigger On Value Change:Bool(Value=^MusicEncounter.bIsPaused)
   n56 = TriggerCompare.Clamp:Bool(A=^MusicEncounter.bIsPaused, Compare=@TriggerOnValueChange_0.Trigger) {also {defaults {B="<MetaSoundBuilderResult.SUCCEEDED: 0>"}}}
   ADSREnvelopeADSREnvelope_0 = ADSR Envelope.ADSR Envelope:Float(
    Release Time=^MusicEncounter.PauseFadeOutTime
    Trigger Attack=^UE.Source.OnPlay
    Trigger Release=VariableAccessor.Trigger(Variable=VariableMutator.Trigger(Variable=@n55.Variable, Value=@n56.True).Variable).Value
   ) {also {defaults {Sustain Level=1}}}
   DecibelstoLinearGain_0 = UE.Decibels to Linear Gain:Float() {also {defaults {Decibels=2}}}
   n57 = UE.Multiply:Float(PrimaryOperand=@ADSREnvelopeADSREnvelope_0.Out Envelope, AdditionalOperands=@DecibelstoLinearGain_0.Linear Gain)
   Multiply_0 = UE.Multiply:Audio by Float(PrimaryOperand=@AudioMixerAudioMixerStereo4_0.Out R, AdditionalOperands=@n57.Out)
   Multiply_1 = UE.Multiply:Audio by Float(PrimaryOperand=@AudioMixerAudioMixerStereo4_0.Out L, AdditionalOperands=@n57.Out)
   n58 = TriggerCompare.Clamp:Int32(A=^MusicEncounter.Boss.Phase, B=VariableAccessor.Int32(Variable=@n10.Variable).Value, Compare=@WavePlayer_4.On Finished)
   TriggerAnyTriggerAny2_1 = "TriggerAny.Trigger Any (2)"(
    In 1=@n58.True
    In 0=VariableAccessor.Trigger(Variable=VariableMutator.Trigger(Value=@ADSREnvelopeADSREnvelope_0.On Done, Variable=@n49.Variable).Variable).Value
   )
   VariableAccessor.WaveAsset:Array(Variable=@n45.Variable)
   VariableAccessor.WaveAsset:Array(Variable=@n5.Variable)
   VariableAccessor.WaveAsset:Array(Variable=@n35.Variable)
   VariableAccessor.WaveTableBankAsset(Variable=InitVariable.WaveTableBankAsset().Variable)
   n59 = &MSP_StereoGain("Gain (Linear)"=VariableAccessor.Float(Variable=@n42.Variable).Value, Left In=@WavePlayer_4.Out Left, Right In=@WavePlayer_4.Out Right)
   VariableMutator.Audio(Variable=InitVariable.Audio().Variable, Value=@n59.Right Out)
   VariableMutator.Audio(Variable=InitVariable.Audio().Variable, Value=@n59.Left Out)
   TriggerOnValueChange_1 = UE.Trigger On Value Change:Int32(Value=^MusicEncounter.Boss.Phase)
   TriggerRouteTriggerRouteInt322_0 = "TriggerRoute.Trigger Route (Int32, 2):Int32"(Value 0=^MusicEncounter.Boss.Phase, Set 1=^UE.Source.OnPlay, Set 0=@TriggerOnValueChange_1.Trigger)
   TriggerSelect2_0 = "UE.Trigger Select (2)"(In=@TriggerRouteTriggerRouteInt322_0.On Set, Index=@TriggerRouteTriggerRouteInt322_0.Value)
   TriggerRouteTriggerRouteFloat2_0 = "TriggerRoute.Trigger Route (Float, 2):Float"(Set 0=@TriggerSelect2_0.Out 0, Set 1=@TriggerSelect2_0.Out 1) {also {defaults {Value 0=14, Value 1=13}}}
   VariableMutator.Time(Value=UE.BPMToSeconds(BPM=^Tempo, Beat Multiplier=@TriggerRouteTriggerRouteFloat2_0.Value).Seconds, Variable=InitVariable.Time().Variable)
   VariableMutator.Trigger(Value=@ADSREnvelopeADSREnvelope_2.On Done, Variable=@n28.Variable)
   VariableMutator.Trigger(Value=@n58.False, Variable=@n53.Variable)
   VariableMutator.Trigger(Value=@ADSREnvelopeADSREnvelope_1.On Done, Variable=@n14.Variable)
   TriggerAnyTriggerAny2_3 = "TriggerAny.Trigger Any (2)"(In 0=@n23.Value, In 1=@n26.Value)
   n60 = VariableDeferredAccessor.Trigger(Variable=InitVariable.Trigger().Variable)
   TriggerControl_0 = UE.Trigger Control(
    Trigger In=HarmonixNodes.MidiClockSubdivisionTrigger(MIDI Clock=@n16.Value) {also {defaults {Base Grid Size=5}}}.Trigger Out
    Open=@TriggerAnyTriggerAny2_3.Out
    Close=@n60.Value
   )
   TriggerCounter_0 = UE.Trigger Counter(In=@TriggerControl_0.Trigger Out) {also {defaults {Reset Count=4}}}
   n61 = VariableDeferredAccessor.Trigger(Variable=@n60.Variable)
   TriggerOnce_2 = UE.Trigger Once(Trigger In=@TriggerCounter_0.On Reset, Reset=@n61.Value)
   VariableMutator.Trigger(Value=@TriggerOnce_2.Trigger Out, Variable=@n61.Variable)
  ]
  links [
   {dst="@UE.OutputFormat.Stereo.Audio:1", src=Multiply_0.Out}
   {dst="@UE.OutputFormat.Stereo.Audio:0", src=Multiply_1.Out}
   {dst="@UE.Source.OneShot.OnFinished", src=TriggerAnyTriggerAny2_1.Out}
  ]
 }]
properties {
 assetImportData=&self:AssetImportData
 numChannels=2
 outputFormat=Stereo
 sampleRate=48000
 soundClassObject=&Fort_Music_NPC
 volume=0.75
 busSends [{audioBus=&AB_Sidechain_FortAmbient_Vol, sendLevel=0.5}]
}
